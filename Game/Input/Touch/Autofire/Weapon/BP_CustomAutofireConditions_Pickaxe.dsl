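; Custom autofire condition for the pickaxe under touch input, deciding whether a target allows automatic swinging.
; ShouldAllowAutofire requires a valid target with a valid parent actor, and that parent must carry the slurp barrel container gameplay tag.
; The tag result is ANDed with the parent class's check on the original weapon and target, so it only narrows the inherited rule.
; All other cases return false.
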(event EventBeginPlay)

(event EventTick
  (params (DeltaSeconds :type Float%20%28single-precision%29)))

(fn ShouldAllowAutofire
  (params (Weapon :type Fort%20Weapon%20Object%20Reference) (Target :type Actor%20Object%20Reference))
  (bind _returnvalue (Actor|GetParentActor Target))
  (Utilities|IsValid Target
    (:Is%20Valid
      (Utilities|IsValid _returnvalue
        (:Is%20Valid
          (bind _returnvalue_2 (GameplayTags|HasMatchingGameplayTag(Message) _returnvalue "(TagName=\"Building.Type.Container.Slurp.Barrel\")"))
          (return (and _returnvalue_2 (|CallParentFunction Weapon Target))))
        (:Is%20Not%20Valid
          _)))
    (:Is%20Not%20Valid
      _))
  (return false))
